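# Swing sound cue for the Frogman Female one-handed pickaxe cosmetic.
# A mixer layers Frogman swing waves under close attenuation with shared DefaultV2 distant swing waves under medium attenuation.
# In each layer, quality-level nodes cut random variations down to a single wave at the lower tiers, with modulators varying volume.
# Concurrency overrides limit copies per owner, stopping the oldest, and the cue primes on load.
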SoundCue /BRCosmetics/Sounds/PickAxes/FrogmanFemale1H/Frogman_Female1H_Swing_Cue
root=Mixer_1
nodes [
 {id=Attenuation_2, children [Modulator_3], properties {attenuationSettings=/Game/Sounds/Attenuation/Fort_Weapon_Pickaxe_Close_Att}}
 {id=Attenuation_3, children [Modulator_2], properties {attenuationSettings=/Game/Sounds/Attenuation/Fort_Weapon_Pickaxe_Medium_Att}}
 {id=Mixer_1, children [Attenuation_2, Attenuation_3], properties {inputVolume [1,1]}}
 {id=Modulator_10, children [Random_9], properties {volumeMax=0.5, volumeMin=0.4}}
 {id=Modulator_11, children [Random_10], properties {volumeMax=0.5, volumeMin=0.4}}
 {id=Modulator_12, children [WavePlayer_52], properties {pitchMax=1.1, pitchMin=0.9, volumeMax=0.5, volumeMin=0.4}}
 {id=Modulator_2, children [QualityLevel_3]}
 {id=Modulator_3, children [QualityLevel_0]}
 {id=Modulator_7, children [Random_6], properties {volumeMax=0.9, volumeMin=0.8}}
 {id=Modulator_8, children [Random_7], properties {volumeMax=0.9, volumeMin=0.8}}
 {id=Modulator_9, children [WavePlayer_37], properties {volumeMax=0.9, volumeMin=0.8}}
 {id=QualityLevel_0, children [Modulator_10,Modulator_11,Modulator_12,Modulator_12]}
 {id=QualityLevel_3, children [Modulator_7,Modulator_8,Modulator_9,Modulator_9]}
 {id=Random_10, children [WavePlayer_49, WavePlayer_50, WavePlayer_51], properties {weights [1,1,1]}}
 {id=Random_6, children [WavePlayer_29, WavePlayer_30, WavePlayer_31, WavePlayer_32, WavePlayer_33], properties {weights [1,1,1,1,1]}}
 {id=Random_7, children [WavePlayer_34, WavePlayer_35, WavePlayer_36], properties {weights [1,1,1]}}
 {
  id=Random_9
  children [WavePlayer_43, WavePlayer_44, WavePlayer_45, WavePlayer_46, WavePlayer_47, WavePlayer_48]
  properties {weights [1,1,1,1,1,1]}
 }
 {id=WavePlayer_29, properties {soundWaveAssetPtr=/BRCosmetics/Sounds/PickAxes/DefaultV2/Pickaxe_DefaultV2_Swing_Distant_01}}
 {id=WavePlayer_30, properties {soundWaveAssetPtr=/BRCosmetics/Sounds/PickAxes/DefaultV2/Pickaxe_DefaultV2_Swing_Distant_02}}
 {id=WavePlayer_31, properties {soundWaveAssetPtr=/BRCosmetics/Sounds/PickAxes/DefaultV2/Pickaxe_DefaultV2_Swing_Distant_03}}
 {id=WavePlayer_32, properties {soundWaveAssetPtr=/BRCosmetics/Sounds/PickAxes/DefaultV2/Pickaxe_DefaultV2_Swing_Distant_04}}
 {id=WavePlayer_33, properties {soundWaveAssetPtr=/BRCosmetics/Sounds/PickAxes/DefaultV2/Pickaxe_DefaultV2_Swing_Distant_05}}
 {id=WavePlayer_34, properties {soundWaveAssetPtr=&Pickaxe_DefaultV2_Swing_Distant_01}}
 {id=WavePlayer_35, properties {soundWaveAssetPtr=&Pickaxe_DefaultV2_Swing_Distant_04}}
 {id=WavePlayer_36, properties {soundWaveAssetPtr=&Pickaxe_DefaultV2_Swing_Distant_05}}
 {id=WavePlayer_37, properties {soundWaveAssetPtr=&Pickaxe_DefaultV2_Swing_Distant_04}}
 {id=WavePlayer_43, properties {soundWaveAssetPtr=&/Frogman_Female1H_Swing_01}}
 {id=WavePlayer_44, properties {soundWaveAssetPtr=&/Frogman_Female1H_Swing_02}}
 {id=WavePlayer_45, properties {soundWaveAssetPtr=&/Frogman_Female1H_Swing_03}}
 {id=WavePlayer_46, properties {soundWaveAssetPtr=&/Frogman_Female1H_Swing_04}}
 {id=WavePlayer_47, properties {soundWaveAssetPtr=&/Frogman_Female1H_Swing_05}}
 {id=WavePlayer_48, properties {soundWaveAssetPtr=&/Frogman_Female1H_Swing_06}}
 {id=WavePlayer_49, properties {soundWaveAssetPtr=&Frogman_Female1H_Swing_01}}
 {id=WavePlayer_50, properties {soundWaveAssetPtr=&Frogman_Female1H_Swing_02}}
 {id=WavePlayer_51, properties {soundWaveAssetPtr=&Frogman_Female1H_Swing_05}}
 {id=WavePlayer_52, properties {soundWaveAssetPtr=&Frogman_Female1H_Swing_02}}
]
properties {
 audioPropertiesSheet=/BRCosmetics/Sounds/Settings/PropertySheets/Pickaxe/APS_Cosmetic_Pickaxe_Swing
 bPrimeOnLoad=true
 maxDistance=3000
 soundClassObject=&Fort_WeaponMelee_PSM
 volumeMultiplier=0.3
 concurrencyOverrides {bLimitToOwner=true, maxCount=2, resolutionRule=StopOldest}
 concurrencySet [&Player_Melee_Weapon_Conc]
}
